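# Unlit additive Niagara ribbon material for the Coyote Trail glider's trail.
# Color blends panning smoke noise with a curve-atlas gradient along V, tinted by particle color and optionally time of day.
# Opacity multiplies particle alpha, a panning primary mask, a powered secondary mask, camera depth fade and an optional front fade along U.
# Optional depth fade softens intersections, but the mobile shading path skips it.
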Material /BRCosmetics/Gadgets/Assets/VinderTech_GliderChute/Glider_CoyoteTrail/FX/M_Glider_CoyoteTrail_Trail
graph {
 properties {
  bAutomaticallySetUsageInEditor=false
  bUsedWithNiagaraRibbons=true
  bUsedWithStaticMesh=false
  blendMode=BLEND_Additive
  pixelDepthOffsetMode=PDOM_Legacy
  shadingModel=MSM_Unlit
  usedShadingModels=Unlit
 }
 nodes [
  n1 = MaterialFunctionCall(
   Color_Morning=VectorParameter() {parameterName=MorningColor}.RGB
   Color_Day=VectorParameter() {parameterName=DayColor}.RGB
   Color_Evening=VectorParameter() {parameterName=EveningColor}.RGB
   Color_Night=VectorParameter() {parameterName=NightColor}.RGB
  ) {materialFunction=&Color_Time_of_Day}
  n2 = ParticleColor()
  n3 = ScalarParameter() {defaultValue=0.1, parameterName=CloudTileU}
  n4 = ScalarParameter() {defaultValue=1, parameterName=CloudTileV}
  n5 = Panner(
   Coordinate=Multiply(A=TextureCoordinate() {coordinateIndex=1}, B=AppendVector(A=@n3, B=@n4))
   Speed=AppendVector(A=ScalarParameter() {parameterName=TilingNoisePanSpeedU}, B=ScalarParameter() {parameterName=TilingNoisePanSpeedV})
  ) {desc=Default Speed is 0}
  n6 = CurveAtlasRowParameter(CurveTime=ComponentMask(None=TextureCoordinate()) {g=true}) {
   atlas=/Game/Athena/LinearColorCurves/A_Atlas_Turtleneck_Glider
   curve=/Game/Gadgets/Assets/VinderTech_GliderChute/Glider_CoyoteTrail/FX/Curve_Glider_CoyoteTrail
   defaultValue=43
  }
  n7 = Multiply(
   A=@n2.RGB
   B=LinearInterpolate(
    A=TextureSampleParameter2D(UVs=@n5) {parameterName=TilingNoise, texture=/Game/Global/Textures/Default/Tileable_Patterns/T_Smoke_Puff_Tile_Up}.G
    B=@n6
    Alpha=ScalarParameter() {parameterName=Use Curve Color}
   )
  )
  MaterialFunctionCall(Emissive=StaticSwitchParameter(True=Multiply(A=@n1.Result, B=@n7), False=@n7) {parameterName=Modulate By Time of Day}) {materialFunction=&MF_FX_EmissiveAdjustments}
  n8 = ScalarParameter() {defaultValue=1, parameterName=PrimaryMaskTileU}
  n9 = ScalarParameter() {defaultValue=1, parameterName=PrimaryMaskTileV}
  n10 = Add(
   A=Multiply(A=TextureCoordinate(), B=AppendVector(A=@n8, B=@n9))
   B=AppendVector(A=ScalarParameter() {defaultValue=1, parameterName=PrimaryMaskAddU}, B=ScalarParameter() {defaultValue=1, parameterName=PrimaryMaskAddV})
  )
  n11 = Panner(
   Coordinate=StaticSwitchParameter(True=MaterialFunctionCall(XY=@n10) {materialFunction=&Swizzle}.YX, False=@n10) {parameterName=SwizzlePrimaryMaskUVs}
   Speed=AppendVector(A=ScalarParameter() {parameterName=PrimaryMaskPanSpeedU}, B=ScalarParameter() {parameterName=PrimaryMaskPanSpeedV})
  ) {desc=Default Speed is 0}
  n12 = Multiply(
   A=TextureCoordinate()
   B=AppendVector(
    A=ScalarParameter() {defaultValue=1, parameterName=SecondaryMaskTileU}
    B=ScalarParameter() {defaultValue=1, parameterName=SecondaryMaskTileV}
   )
  )
  n13 = ComponentMask(None=@n12) {g=true}
  n14 = OneMinus(None=@n13)
  n15 = ScalarParameter() {parameterName=FlipSecondaryMask_V}
  n16 = TextureSampleParameter2D(UVs=AppendVector(A=ComponentMask(None=@n12) {r=true}, B=LinearInterpolate(A=@n13, B=@n14, Alpha=@n15))) {
   parameterName=SecondaryMask
   samplerSource=SSM_Wrap_WorldGroupSettings
   texture=/Game/Effects/Fort_Effects/Textures/Melee/T_GenericWeaponTrail
  }
  n17 = ScalarParameter() {defaultValue=0.5, parameterName=SecondaryMaskPower}
  n18 = Clamp(None=Multiply(
    A=TextureSampleParameter2D(UVs=@n11) {parameterName=PrimaryMask, texture=/Game/Global/Textures/Default/Masks_and_Gradients/T_Water_Top_Ramp}.G
    B=Power(Base=@n16.G, Exp=@n17) {constExponent=0.5}
   ))
  n19 = ScalarParameter() {defaultValue=150, parameterName=CamDepthFadeLength}
  n20 = Multiply(A=ComponentMask(None=TextureCoordinate()) {r=true}) {constB=12}
  n21 = Multiply(A=@n20, B=@n20)
  n22 = Multiply(
   A=Multiply(
    A=Multiply(A=@n2.A, B=@n18)
    B=MaterialFunctionCall(Fade Length=@n19) {materialFunction=&CameraDepthFade}.Result
   )
   B=StaticSwitchParameter(True=Clamp(None=@n21), False=Constant() {r=1}) {defaultValue=true, parameterName=IncludeFrontFade}
  )
  n23 = ScalarParameter() {parameterName=DepthFadeDistance}
  ShadingPathSwitch(Default=StaticSwitchParameter(True=DepthFade(Opacity=@n22, FadeDistance=@n23), False=@n22) {parameterName=UseDepthFade}, Mobile=@n22) {desc=Migrated from FeatureLevelSwitch}
 ]
}
