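# Niagara effect for the Coyote Trail glider cosmetic, adding ribbon trails, flying feathers and a wind mesh burst.
# Two AnimTrail ribbon emitters follow skeletal-mesh bone pairs, their gameplay opacity scaled by an integrator value and time of day.
# Delayed feather bursts spawn at two side offsets, while drop and fly emitters shed mesh feathers from a second skeletal mesh.
# Front-end modifiers swap burst emitters and retune motion and opacity for menu display.
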NiagaraSystem /BRCosmetics/Gadgets/Assets/VinderTech_GliderChute/Glider_CoyoteTrail/FX/NS_Glider_CoyoteTrail
dependencies {
 usedDataInterfaces [&NiagaraDataInterfaceCurve, &NiagaraDataInterfaceColorCurve, &NiagaraDataInterfaceArrayFloat3]
 usedDynamicInputs []
 usedModules [&SystemState, &FN_Integrator]
}
emitters {
 FN_AnimTrail {
  emitterSpawnScript [&FN_RibbonIDOffsetUtility()]
  emitterUpdateScript [
   &EmitterState(Life Cycle Mode=::NewEnumerator1, Loop Delay=&Add_Float(A=$User.Delay, B=0.05), Loop Duration=1, UseLoopDelay=true)
   &FN_SpawnAnimTrail(Number of Anim Trails=2)
  ]
  particleSpawnScript [
   &FN_InitializeAnimTrailBoneIndecies()
   &FN_InitializeAnimTrailBoneIndecies(AnimTrail Selection=1, First Bone or Socket Index=2, Second Bone or Socket Index=3) {module=FN_InitializeAnimTrailBoneIndecies001}
   &FN_InitializeAnimTrail(Lifetime=0.25)
   &AddVelocity(Velocity=&FN_FrontEndVector3Modifier(Gameplay=<0 0 0>, Front End=<0 1000 1000>))
  ]
  particleUpdateScript [
   &SolveForcesAndVelocity(Clamp Velocity=false, Limit Acceleration=false)
   &Color(Color=&MakeLinearColorFromVectorAndFloat(
     "Vector (RGB)"=<1 1 1>
     "Float (Alpha)"=&Multiply_Float(
      A=&FloatFromCurve(
       FloatCurve=DataInterface() {propertyValues {Curve {keys [
           {arriveTangent=-1.5186, interpMode=RCIM_Cubic, leaveTangent=-1.5186, tangentMode=RCTM_User, time=0, value=0.8}
           {arriveTangent=-0.6904, interpMode=RCIM_Cubic, leaveTangent=-0.6904, tangentMode=RCTM_User, time=0.3, value=0.25}
           {interpMode=RCIM_Cubic, tangentMode=RCTM_User, time=1, value=0}
          ]}}}
       Scale Curve=2
      )
      B=&FN_FrontEndFloatModifier(
       Gameplay=1
       Front End=&FloatFromCurve(FloatCurve=DataInterface() {propertyValues {Curve {keys [<at 1 0>,<at 1.1 1>]}}}, CurveIndex=$Emitter.Age)
      )
     )
    ))
   &FN_UpdateAnimTrail(Locked Particle RibbonWidth=1, SkeletalMesh=$User.SkeletalMesh) {module=FN_UpdateAnimTrail001}
   /SetVariables(Particles.RibbonU0Override=$Particles.NormalizedAge) {isSetParameters=true, module=SetVariables_#1}
   &ScaleColor(
    Color Value To Scale=$Particles.Color
    Scale Alpha=&Multiply_Float(
     A=&FN_FrontEndFloatModifier(Gameplay=$System.FN_Integrator.Value, Front End=0.35)
     B=&FNTimeOfDayScalar(Night Scale=0.5, Morning Scale=0.3, Evening Scale=0.5)
    )
    ScaleA=true
    ScaleRGB=true
   )
  ]
  rendererClasses [&NiagaraRibbonRendererProperties]
  renderers [{rendererClass=&NiagaraRibbonRendererProperties}]
 }
 FN_AnimTrail001 {
  emitterSpawnScript [&FN_RibbonIDOffsetUtility()]
  emitterUpdateScript [
   &EmitterState(Life Cycle Mode=::NewEnumerator1, Loop Delay=$User.Delay, Loop Duration=1, UseLoopDelay=true)
   &FN_SpawnAnimTrail(Number of Anim Trails=2)
  ]
  particleSpawnScript [
   &FN_InitializeAnimTrailBoneIndecies()
   &FN_InitializeAnimTrailBoneIndecies(AnimTrail Selection=1, First Bone or Socket Index=2, Second Bone or Socket Index=3) {module=FN_InitializeAnimTrailBoneIndecies001}
   &FN_InitializeAnimTrail(AnimTrailWidth=1.02, Lifetime=0.25)
   &AddVelocity(Velocity=&FN_FrontEndVector3Modifier(Gameplay=<0 0 0>, Front End=<0 1000 1000>))
  ]
  particleUpdateScript [
   &SolveForcesAndVelocity(Clamp Velocity=false, Limit Acceleration=false)
   &Color(Color=&MakeLinearColorFromVectorAndFloat(
     "Vector (RGB)"=<1 1 1>
     "Float (Alpha)"=&Multiply_Float(
      A=&FloatFromCurve(
       FloatCurve=DataInterface() {propertyValues {Curve {keys [
           {arriveTangent=-1.5186, interpMode=RCIM_Cubic, leaveTangent=-1.5186, tangentMode=RCTM_User, time=0, value=0.8}
           {arriveTangent=-0.6904, interpMode=RCIM_Cubic, leaveTangent=-0.6904, tangentMode=RCTM_User, time=0.3, value=0.25}
           {interpMode=RCIM_Cubic, tangentMode=RCTM_User, time=1, value=0}
          ]}}}
       Scale Curve=0.25
      )
      B=&FN_FrontEndFloatModifier(
       Gameplay=1
       Front End=&FloatFromCurve(FloatCurve=DataInterface() {propertyValues {Curve {keys [<at 1 0>,<at 1.1 1>]}}}, CurveIndex=$Emitter.Age)
      )
     )
    ))
   &FN_UpdateAnimTrail(Locked Particle RibbonWidth=1, SkeletalMesh=$User.SkeletalMesh) {module=FN_UpdateAnimTrail001}
   /SetVariables(Particles.RibbonU0Override=$Particles.NormalizedAge) {isSetParameters=true, module=SetVariables_#1}
   &ScaleColor(
    Color Value To Scale=$Particles.Color
    Scale Alpha=&Multiply_Float(
     A=&FN_FrontEndFloatModifier(Gameplay=$System.FN_Integrator.Value, Front End=0.35)
     B=&FNTimeOfDayScalar(Night Scale=0.5, Morning Scale=0.3, Evening Scale=0.5)
    )
    ScaleA=true
    ScaleRGB=true
   )
  ]
  rendererClasses [&NiagaraRibbonRendererProperties]
  renderers [{rendererClass=&NiagaraRibbonRendererProperties}]
 }
 Feathers_Burst001 {
  emitterUpdateScript [
   &EmitterState(
    Inactive Response=::NewEnumerator2
    Life Cycle Mode=::NewEnumerator1
    Loop Behavior=::NewEnumerator1
    Loop Delay=&Add_Float(A=0.9, B=0.1)
    Loop Duration=0.4
    UseLoopDelay=true
   )
   &SpawnBurst_Instantaneous(
    Spawn Count=16
    Spawn Probability=&FN_FrontEndFloatModifier(Gameplay=1, Front End=0)
    Use Spawn Probability=true
    UseLoopCountLimit=false
   )
   &SpawnRate(
    Spawn Probability=&FN_FrontEndFloatModifier(Gameplay=1, Front End=0)
    SpawnRate=20
    Use Spawn Probability=true
   )
  ]
  particleSpawnScript [
   &FN_SetSpawnGroup(Normalize Spawn Group=true, Number of Spawn Groups=2, Particles Per Group=6)
   &InitializeParticle(
    Color Mode=::NewEnumerator0
    Lifetime Max=0.45
    Lifetime Min=0.35
    Lifetime Mode=::NewEnumerator1
    Mesh Scale Mode=::NewEnumerator1
    Mesh Uniform Scale Max=1.6
    Mesh Uniform Scale Min=1.3
    Position Offset=&SelectVectorFromArray(Direct Array Index=$Particles.SpawnGroup, Vector Selection Array=DataInterface() {propertyValues {FloatData [<65 20 -10>, <-65 20 -10>]}})
    Position Offset Coordinate Space=::Local
    UsePositionOffset=true
    Write Material Random=true
   )
   &InitialMeshOrientation(Use Rotation Vector=false)
   &InheritVelocity(Module.Limit Speed=false, Velocity Scale=<0.01 0.01 0.01>) {enabled=false}
   &AddVelocity(Velocity=&FN_FrontEndVector3Modifier(Gameplay=<0 -100 150>, Front End=<0 0 0>))
   &AddVelocityFromPoint(Module.Use Falloff Distance=false, Velocity Strength=&RandomRangeFloat(Minimum=160, Maximum=100))
   &SphereLocation(
    MaskSpawnGroup=false
    Override Seed=false
    Sphere Radius=50
    Surface Only=false
    UseNonUniformScale=false
    UseOffset=false
    UseSphereOrientation=false
   )
   /SetVariables(Particles.MeshIndex=&MakeIntFromFloat(Float=&UniformAOrBFloat(A=0, Distribution Weight=0.8))) {isSetParameters=true, module=SetVariables_#2}
  ]
  particleUpdateScript [
   &Drag(Drag=0.7, Use Linear Drag=true, Use Rotational Drag=false)
   &GravityForce(Gravity=<0 0 -100>)
   &CurlNoiseForce(Noise Strength=1000, Pan Noise=false)
   &InheritSourceMovement(Applied Position Scale=<0.95 0.95 0.95>, ApplyForce=false, ApplyPosition=true, ApplyPositionFalloff=false)
   &SolveForcesAndVelocity(Clamp Velocity=false, Limit Acceleration=false)
   &ParticleState()
   &DynamicMaterialParameters(
    Index 0 Param 1=&RandomRangeFloat(Minimum=0.1, Maximum=0.2)
    Index 0 Param 2=&FloatFromCurve(FloatCurve=DataInterface() {propertyValues {Curve {keys [<at 0.6 1>,<at 1 0>]}}})
    Index 0 Param 3=&FloatFromCurve(FloatCurve=DataInterface() {propertyValues {Curve {keys [<at 0 1>,<at 0.5 10>]}}})
    Index 0 Param 4=&RandomRangeFloat(Minimum=0, Maximum=0.5)
    Module.Param0WriteEnabled=true
    Module.Param1WriteEnabled=true
    Module.Param2WriteEnabled=true
    Module.Param3WriteEnabled=true
   )
   &ScaleColor(Scale RGB=&Multiply_VectorByFloat(Vector=<1 1 1>, Float=1.3), ScaleA=true, ScaleRGB=true)
   &UpdateMeshOrientation(Rotation Vector=&VectorFromFloat(Value=&RandomRangeFloat(Minimum=-1, Maximum=1)))
   &ScaleMeshSize(Scale Factor=&Multiply_VectorByFloat(
     Vector=<1 1 1>
     Float=&FloatFromCurve(FloatCurve=DataInterface() {propertyValues {Curve {keys [
          {arriveTangent=0.1449, interpMode=RCIM_Cubic, leaveTangent=0.1449, tangentMode=RCTM_User, time=0, value=0}
          {arriveTangent=0.9995, interpMode=RCIM_Cubic, leaveTangent=-0.1268, tangentMode=RCTM_Break, time=0.4, value=1}
          {arriveTangent=-1.8165, interpMode=RCIM_Cubic, leaveTangent=-1.8165, tangentMode=RCTM_User, time=0.6, value=1}
          {interpMode=RCIM_Cubic, leaveTangent=-1.5153, time=1, value=0}
         ]}}})
    )) {module=ScaleMeshSize001}
  ]
  rendererClasses [&NiagaraMeshRendererProperties]
  renderers [{rendererClass=&NiagaraMeshRendererProperties, rendererIndex=0}]
 }
 Feathers_Burst_FE {
  emitterUpdateScript [
   &EmitterState(
    Inactive Response=::NewEnumerator2
    Life Cycle Mode=::NewEnumerator1
    Loop Behavior=::NewEnumerator1
    Loop Delay=&Add_Float(A=0.9, B=&FN_FrontEndFloatModifier(Gameplay=0.05, Front End=0.25))
    Loop Duration=0.4
    UseLoopDelay=true
   )
   &SpawnBurst_Instantaneous(
    Spawn Count=16
    Spawn Probability=&FN_FrontEndFloatModifier(Gameplay=0, Front End=1)
    Use Spawn Probability=true
    UseLoopCountLimit=false
   )
   &SpawnRate(
    Spawn Probability=&FN_FrontEndFloatModifier(Gameplay=0, Front End=1)
    SpawnRate=20
    Use Spawn Probability=true
   )
  ]
  particleSpawnScript [
   &FN_SetSpawnGroup(Normalize Spawn Group=true, Number of Spawn Groups=2, Particles Per Group=6)
   &InitializeParticle(
    Color Mode=::NewEnumerator0
    Lifetime Max=0.7
    Lifetime Min=0.4
    Lifetime Mode=::NewEnumerator1
    Mesh Scale Mode=::NewEnumerator1
    Mesh Uniform Scale Max=1.6
    Mesh Uniform Scale Min=1.3
    Position Offset=&SelectVectorFromArray(Direct Array Index=$Particles.SpawnGroup, Vector Selection Array=DataInterface() {propertyValues {FloatData [<70 0 -20>, <-70 0 -20>]}})
    Position Offset Coordinate Space=::Local
    UsePositionOffset=true
    Write Material Random=true
   )
   &InitialMeshOrientation(Use Rotation Vector=false)
   &InheritVelocity(Module.Limit Speed=false, Velocity Scale=<0.01 0.01 0.01>)
   &AddVelocity(Velocity=&FN_FrontEndVector3Modifier(Gameplay=<0 -100 200>, Front End=<0 0 0>))
   &AddVelocityFromPoint(Module.Use Falloff Distance=false, Velocity Strength=&RandomRangeFloat(Minimum=70, Maximum=120))
   &SphereLocation(
    MaskSpawnGroup=false
    Override Seed=false
    Sphere Radius=50
    Surface Only=false
    UseNonUniformScale=false
    UseOffset=false
    UseSphereOrientation=false
   )
   /SetVariables(Particles.MeshIndex=&MakeIntFromFloat(Float=&UniformAOrBFloat(A=0, Distribution Weight=0.8))) {isSetParameters=true, module=SetVariables_#2}
  ]
  particleUpdateScript [
   &Drag(Drag=0.2, Use Linear Drag=true, Use Rotational Drag=false)
   &GravityForce(Gravity=<0 0 -100>)
   &CurlNoiseForce(Noise Strength=1000, Pan Noise=false)
   &SolveForcesAndVelocity(Clamp Velocity=false, Limit Acceleration=false)
   &ParticleState()
   &DynamicMaterialParameters(
    Index 0 Param 1=&RandomRangeFloat(Minimum=0.1, Maximum=0.2)
    Index 0 Param 2=&FloatFromCurve(FloatCurve=DataInterface() {propertyValues {Curve {keys [<at 0.6 1>,<at 1 0>]}}})
    Index 0 Param 3=&FloatFromCurve(FloatCurve=DataInterface() {propertyValues {Curve {keys [<at 0 1>,<at 0.5 10>]}}})
    Index 0 Param 4=&RandomRangeFloat(Minimum=0, Maximum=0.5)
    Module.Param0WriteEnabled=true
    Module.Param1WriteEnabled=true
    Module.Param2WriteEnabled=true
    Module.Param3WriteEnabled=true
   )
   &ScaleColor(Scale RGB=&Multiply_VectorByFloat(Vector=<1 1 1>, Float=1.3), ScaleA=true, ScaleRGB=true)
   &UpdateMeshOrientation(Rotation Vector=&VectorFromFloat(Value=&RandomRangeFloat(Minimum=-1, Maximum=1)))
   &ScaleMeshSize(Scale Factor=&Multiply_VectorByFloat(
     Vector=<1 1 1>
     Float=&FloatFromCurve(FloatCurve=DataInterface() {propertyValues {Curve {keys [
          {arriveTangent=4.1785, interpMode=RCIM_Cubic, leaveTangent=4.1785, tangentMode=RCTM_User, time=0, value=0}
          {arriveTangent=0.9995, interpMode=RCIM_Cubic, leaveTangent=-0.1394, tangentMode=RCTM_Break, time=0.15, value=1}
          {arriveTangent=-1.1765, interpMode=RCIM_Cubic, leaveTangent=-1.1765, time=0.5, value=1}
          {interpMode=RCIM_Cubic, leaveTangent=-1.5153, time=1, value=0}
         ]}}})
    )) {module=ScaleMeshSize001}
  ]
  rendererClasses [&NiagaraMeshRendererProperties]
  renderers [{rendererClass=&NiagaraMeshRendererProperties, rendererIndex=0}]
 }
 Feathers_Drop {
  emitterUpdateScript [
   &EmitterState(
    Life Cycle Mode=::NewEnumerator1
    Loop Behavior=::NewEnumerator1
    Loop Delay=&Add_Float(A=0.9, B=&FN_FrontEndFloatModifier(Gameplay=0.05, Front End=0.15))
    Loop Duration=0.5
    UseLoopDelay=true
   )
   &SpawnRate(
    SpawnGroup=1
    SpawnRate=&FN_FrontEndFloatModifier(Gameplay=18, Front End=10)
    Use Spawn Probability=false
   ) {module=SpawnRate001}
  ]
  particleSpawnScript [
   &InitializeParticle(
    Color Mode=::NewEnumerator0
    Lifetime Max=1.2
    Lifetime Min=0.7
    Lifetime Mode=::NewEnumerator1
    Mesh Scale Mode=::NewEnumerator3
    Mesh Uniform Scale=&RandomRangeFloat(Minimum=1.3, Maximum=1.6)
    UsePositionOffset=false
    Write Material Random=true
   )
   &SkeletalMeshLocation(
    Bone Sampling Mode=::NewEnumerator0
    Inherited Velocity Speed Limit=150
    MaskSpawnGroup=false
    Orientation Sampling=::NewEnumerator1
    OverrideRandomSeed=false
    Skeletal Mesh=$User.SkeletalMesh02
    Velocity Sampling=::NewEnumerator0
    VelocitySpeedLimit=true
   )
   &FN_SetSpawnGroup(Normalize Spawn Group=true, Number of Spawn Groups=6, Particles Per Group=1)
   &AddVelocity(
    Coordinate Space=::World
    Velocity=&Multiply_VectorByFloat(
     Vector=<0 0 100>
     Float=&FloatFromCurve(FloatCurve=DataInterface() {propertyValues {Curve {keys [{tangentMode=RCTM_User, time=0, value=-1}, {tangentMode=RCTM_User, time=1, value=1}]}}})
    )
   )
   &AddVelocityFromPoint(Module.Use Falloff Distance=false, Velocity Strength=&RandomRangeFloat(Minimum=250, Maximum=300))
   /SetVariables(Particles.MeshIndex=&MakeIntFromFloat(Float=&UniformAOrBFloat(A=0, Distribution Weight=0.8))) {isSetParameters=true}
  ]
  particleUpdateScript [
   &ParticleState()
   &GravityForce(Gravity=&FN_FrontEndVector3Modifier(Gameplay=<0 0 -100>, Front End=<0 -100 -100>))
   &Drag(Drag=2, Use Linear Drag=true, Use Rotational Drag=false)
   &CurlNoiseForce(
    Noise Frequency=25
    Noise Strength=&RandomRangeFloat(Minimum=200, Maximum=150)
    Pan Noise=true
    Pan Noise Field=<0 0 -1>
   )
   &InheritSourceMovement(
    Applied Position Falloff Distance=500
    Applied Position Scale=&FN_FrontEndVector3Modifier(Gameplay=<0.6 0.6 0.6>, Front End=<0 0 0>)
    ApplyForce=false
    ApplyPosition=true
    ApplyPositionFalloff=false
   )
   &SolveForcesAndVelocity(Clamp Velocity=false, Limit Acceleration=false)
   &ScaleMeshSize(Scale Factor=&Multiply_VectorByFloat(
     Vector=<1 1 1>
     Float=&FloatFromCurve(
      FloatCurve=DataInterface() {propertyValues {Curve {keys [
          {arriveTangent=0.1449, interpMode=RCIM_Cubic, leaveTangent=0.1449, tangentMode=RCTM_User, time=0, value=0}
          {arriveTangent=0.9995, interpMode=RCIM_Cubic, leaveTangent=-0.1268, tangentMode=RCTM_Break, time=0.4, value=1}
          {arriveTangent=-1.8165, interpMode=RCIM_Cubic, leaveTangent=-1.8165, tangentMode=RCTM_User, time=0.6, value=1}
          {interpMode=RCIM_Cubic, leaveTangent=-1.5153, time=1, value=0}
         ]}}}
      Scale Curve=0.9
     )
    )) {module=ScaleMeshSize001}
   &UpdateMeshOrientation(Rotation Vector=&VectorFromFloat(Value=&RandomRangeFloat(Minimum=-1, Maximum=1)))
   &DynamicMaterialParameters(
    Index 0 Param 1=&RandomRangeFloat(Minimum=0.1, Maximum=0.2)
    Index 0 Param 2=&FloatFromCurve(FloatCurve=DataInterface() {propertyValues {Curve {keys [<at 0.6 1>,<at 1 0>]}}})
    Index 0 Param 3=&FloatFromCurve(FloatCurve=DataInterface() {propertyValues {Curve {keys [<at 0 1>,<at 0.5 10>]}}})
    Index 0 Param 4=&RandomRangeFloat(Minimum=0, Maximum=0.5)
    Module.Param0WriteEnabled=true
    Module.Param1WriteEnabled=true
    Module.Param2WriteEnabled=true
    Module.Param3WriteEnabled=true
   )
  ]
  rendererClasses [&NiagaraMeshRendererProperties]
  renderers [{rendererClass=&NiagaraMeshRendererProperties, rendererIndex=0}]
 }
 Feathers_Fly {
  emitterUpdateScript [
   &EmitterState(
    Life Cycle Mode=::NewEnumerator1
    Loop Delay=&Add_Float(A=$User.Delay, B=&FN_FrontEndFloatModifier(Gameplay=0.05, Front End=0.1))
    Loop Duration=1
    UseLoopDelay=true
   )
   &SpawnRate(
    SpawnRate=&FN_FrontEndFloatModifier(
     Gameplay=&self:XY_Speed(Amount:NiagaraFloat=&Multiply_Float(A=$System.FN_Integrator.Value, B=10), Speed Greater Than:NiagaraFloat=0)
     Front End=3
    )
    Use Spawn Probability=false
   ) {module=SpawnRate001}
  ]
  particleSpawnScript [
   &InitializeParticle(
    Color Mode=::NewEnumerator0
    Lifetime Max=1.5
    Lifetime Mode=::NewEnumerator1
    Mesh Scale Mode=::NewEnumerator3
    Mesh Uniform Scale=&RandomRangeFloat(Minimum=1.3, Maximum=1.6)
    UsePositionOffset=true
    Write Material Random=true
   )
   &SkeletalMeshLocation(
    Bone Sampling Mode=::NewEnumerator0
    MaskSpawnGroup=false
    OverrideRandomSeed=false
    Sampled Position Offset=&RandomRangeVector(Minimum=<0 0 0>, Maximum=<0 0 0>)
    Skeletal Mesh=$User.SkeletalMesh02
   )
   &AddVelocityFromPoint(Module.Use Falloff Distance=false, Velocity Strength=&RandomRangeFloat(Minimum=150, Maximum=200))
   &AddVelocity(Velocity=&FN_FrontEndVector3Modifier(Gameplay=<0 0 0>, Front End=<0 100 300>))
   &InitialMeshOrientation(Use Rotation Vector=false)
   /SetVariables(Particles.MeshIndex=&MakeIntFromFloat(Float=&UniformAOrBFloat(A=0, Distribution Weight=0.8))) {isSetParameters=true}
  ]
  particleUpdateScript [
   &ParticleState()
   &ScaleMeshSize(Scale Factor=&Multiply_VectorByFloat(
     Vector=<1 1 1>
     Float=&FloatFromCurve(FloatCurve=DataInterface() {propertyValues {Curve {keys [
          {arriveTangent=0.1449, interpMode=RCIM_Cubic, leaveTangent=0.1449, tangentMode=RCTM_User, time=0, value=0}
          {arriveTangent=0.9995, interpMode=RCIM_Cubic, leaveTangent=-0.1268, tangentMode=RCTM_Break, time=0.1, value=1}
          {arriveTangent=-1.8165, interpMode=RCIM_Cubic, leaveTangent=-1.8165, tangentMode=RCTM_User, time=0.6, value=1}
          {interpMode=RCIM_Cubic, leaveTangent=-1.5153, time=1, value=0}
         ]}}})
    )) {module=ScaleMeshSize001}
   &GravityForce(Gravity=&FN_FrontEndVector3Modifier(Gameplay=<0 0 -200>, Front End=<0 0 0>))
   &Drag(
    Drag=&FN_FrontEndFloatModifier(Gameplay=2, Front End=0)
    Use Linear Drag=true
    Use Rotational Drag=false
   )
   &CurlNoiseForce(
    Noise Frequency=25
    Noise Strength=&RandomRangeFloat(Minimum=250, Maximum=500)
    Pan Noise=true
    Pan Noise Field=<0 0 -1>
   )
   &InheritSourceMovement(Applied Position Scale=<0.7 0.7 0.7>, ApplyForce=false, ApplyPosition=true, ApplyPositionFalloff=false)
   &UpdateMeshOrientation(Rotation Vector=&VectorFromFloat(Value=&RandomRangeFloat(Minimum=-1, Maximum=1)))
   &SolveForcesAndVelocity(Clamp Velocity=false, Limit Acceleration=false)
   &DynamicMaterialParameters(
    Index 0 Param 1=&RandomRangeFloat(Minimum=0.1, Maximum=0.2)
    Index 0 Param 2=&FloatFromCurve(FloatCurve=DataInterface() {propertyValues {Curve {keys [<at 0.6 1>,<at 1 0>]}}})
    Index 0 Param 3=&FloatFromCurve(FloatCurve=DataInterface() {propertyValues {Curve {keys [<at 0 1>,<at 0.5 10>]}}})
    Index 0 Param 4=&RandomRangeFloat(Minimum=0, Maximum=0.5)
    Module.Param0WriteEnabled=true
    Module.Param1WriteEnabled=true
    Module.Param2WriteEnabled=true
    Module.Param3WriteEnabled=true
   )
  ]
  rendererClasses [&NiagaraMeshRendererProperties]
  renderers [{rendererClass=&NiagaraMeshRendererProperties, rendererIndex=0}]
 }
 Wind {
  emitterUpdateScript [
   &EmitterState(Life Cycle Mode=::NewEnumerator1, Loop Behavior=::NewEnumerator1, Loop Delay=0.9, Loop Duration=1, UseLoopDelay=true)
   &SpawnBurst_Instantaneous(Spawn Count=7, Use Spawn Probability=false, UseLoopCountLimit=false)
  ]
  particleSpawnScript [
   &InitializeParticle(
    Color=<rgba 0.7 1 0.9 0.4>
    Lifetime=0.8
    Mesh Scale Mode=::NewEnumerator3
    Mesh Uniform Scale=&Multiply_Float(A=0.6, B=&RandomRangeFloat(Minimum=15, Maximum=16))
    UsePositionOffset=false
    Write Material Random=true
   )
   &InitialMeshOrientation(
    Mesh Orientation Mode=::NewEnumerator7
    Rotation=&RandomRangeVector(Minimum=<-1 -0.0625 -0.0625>, Maximum=<1 0.0625 0.0625>)
    Use Rotation Vector=true
   )
  ]
  particleUpdateScript [
   &ParticleState()
   &DynamicMaterialParameters(
    Index 0 Param 1=&FloatFromCurve(FloatCurve=DataInterface() {propertyValues {Curve {keys [{interpMode=RCIM_Cubic, tangentMode=RCTM_User, time=0.5, value=0}, {interpMode=RCIM_Cubic, tangentMode=RCTM_User, time=1, value=1}]}}})
    Index 0 Param 2=&MakeCustomFloatFromBool(
     Bool=$User.IsFrontEnd
     True Float=0
     False Float=&FNTimeOfDayScalar(Night Scale=0.05, Morning Scale=-0.26, Day Scale=0, Evening Scale=-0.35)
    )
    Index 0 Param 3=&MakeCustomFloatFromBool(
     Bool=$User.IsFrontEnd
     True Float=2
     False Float=&FNTimeOfDayScalar(Night Scale=0.5, Morning Scale=0.23, Evening Scale=0.2)
    )
    Index 0 Param 4=&MakeCustomFloatFromBool(
     Bool=$User.IsFrontEnd
     True Float=0
     False Float=&FNTimeOfDayScalar(Night Scale=0.02, Morning Scale=-0.01, Day Scale=0, Evening Scale=-0.01)
    )
    Module.Param0WriteEnabled=true
    Module.Param1WriteEnabled=true
    Module.Param2WriteEnabled=true
    Module.Param3WriteEnabled=true
   )
   /SetVariables(
    Particles.XY:NiagaraFloat=&FloatFromCurve(
     FloatCurve=DataInterface() {propertyValues {Curve {keys [
         {arriveTangent=0.6321, interpMode=RCIM_Cubic, leaveTangent=0.6321, tangentMode=RCTM_User, time=0, value=0}
         {arriveTangent=0.4623, interpMode=RCIM_Cubic, leaveTangent=0.4623, tangentMode=RCTM_User, time=1, value=1}
        ]}}}
     CurveIndex=$Particles.Age
    )
    Particles.Z:NiagaraFloat=&FloatFromCurve(
     FloatCurve=DataInterface() {propertyValues {Curve {keys [
         {arriveTangent=1.8479, interpMode=RCIM_Cubic, leaveTangent=1.8479, tangentMode=RCTM_User, time=0, value=0}
         {arriveTangent=0.9995, interpMode=RCIM_Cubic, leaveTangent=0.9995, tangentMode=RCTM_User, time=0.4, value=0.4}
         {arriveTangent=0.4855, interpMode=RCIM_Cubic, leaveTangent=0.4855, tangentMode=RCTM_User, time=1, value=0.6}
        ]}}}
     CurveIndex=$Particles.Age
    )
   ) {isSetParameters=true}
   &ScaleMeshSize(Scale Factor=&MakeVector(X=$Particles.XY, Y=$Particles.XY, Z=$Particles.Z))
   &UpdateMeshOrientation(Rotation Rate=1.2, Rotation Vector=<0 0 1>)
   &ScaleMeshSize(Scale Factor=&MakeVector(
     X=1
     Y=1
     Z=&FloatFromCurve(
      FloatCurve=DataInterface() {propertyValues {Curve {keys [{interpMode=RCIM_Cubic, tangentMode=RCTM_Break, time=0, value=1}, <at 1 1.5>]}}}
      CurveIndex=$Particles.Age
      Scale Curve=1.2
     )
    )) {module=ScaleMeshSize001}
   &ScaleColor(
    Scale Alpha=&FloatFromCurve(
     FloatCurve=DataInterface() {propertyValues {Curve {keys [
         {arriveTangent=8.2356, interpMode=RCIM_Cubic, leaveTangent=8.2356, tangentMode=RCTM_User, time=0, value=0}
         {arriveTangent=0.1969, interpMode=RCIM_Cubic, leaveTangent=0.1969, tangentMode=RCTM_User, time=0.2223, value=1.5}
         {arriveTangent=-1.9287, interpMode=RCIM_Cubic, leaveTangent=-1.9287, time=0.5, value=0.6}
         {arriveTangent=-0.7636, interpMode=RCIM_Cubic, leaveTangent=-0.7636, tangentMode=RCTM_User, time=1, value=0}
        ]}}}
     Scale Curve=0.5
    )
    Scale RGB=&VectorFromFloat(Value=&FNTimeOfDayScalar(Night Scale=1.5, Morning Scale=1.5, Evening Scale=1.5))
    ScaleA=true
    ScaleRGB=true
   )
   /SetVariables(Particles.New Position:NiagaraPosition=&Add_Vector(
     A=Linked() {linkedVariable {name=Particles.Position, type=NiagaraPosition}}
     B=&FN_FrontEndVector3Modifier(Gameplay=<0 50 0>, Front End=<0 50 20>)
    )) {isSetParameters=true}
  ]
  rendererClasses [&NiagaraMeshRendererProperties]
  renderers [{rendererClass=&NiagaraMeshRendererProperties, rendererIndex=0}]
 }
}
summary {
 systemName=NS_Glider_CoyoteTrail
 emitters [
  {emitterName=FN_AnimTrail, rendererClasses [&NiagaraRibbonRendererProperties]}
  {emitterName=Feathers_Fly, rendererClasses [&NiagaraMeshRendererProperties]}
  {emitterName=Feathers_Drop, rendererClasses [&NiagaraMeshRendererProperties]}
  {emitterName=FN_AnimTrail001, rendererClasses [&NiagaraRibbonRendererProperties]}
  {emitterName=Wind, rendererClasses [&NiagaraMeshRendererProperties]}
  {emitterName=Feathers_Burst_FE, rendererClasses [&NiagaraMeshRendererProperties]}
  {emitterName=Feathers_Burst001, rendererClasses [&NiagaraMeshRendererProperties]}
 ]
 userVariables [
  User.bIsFrontEnd {type=NiagaraBool}
  User.bIsFrontEndPreview {type=NiagaraBool}
  User.bIsFullyDeployed {type=NiagaraBool}
  User.Delay {type=NiagaraFloat, defaultValue=0.3, description=""}
  User.ForwardDot {type=NiagaraFloat, defaultValue=0, description=""}
  User.IsFrontEnd {type=NiagaraBool}
  User.PlayerSpeed {type=NiagaraFloat, defaultValue=0, description=""}
  User.RibbonIDOffset {type=NiagaraInt32}
  User.RightDot {type=NiagaraFloat, defaultValue=0, description=""}
  User.SkeletalMesh {
   type=NiagaraDataInterfaceSkeletalMesh
   defaultValue {
    struct=NiagaraExt_VariableValue_DataInterface
    value {dataInterface=&self:NiagaraDataInterfaceSkeletalMesh_0}
   }
  }
  User.SkeletalMesh02 {
   type=NiagaraDataInterfaceSkeletalMesh
   defaultValue {
    struct=NiagaraExt_VariableValue_DataInterface
    value {dataInterface=&self:NiagaraDataInterfaceSkeletalMesh_1}
   }
  }
 ]
}
